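# Demo surface material for the backpack of the medium male player robot character, Default Lit and flagged for skeletal meshes and mesh deformers.
# Diffuse and normal textures plus constant metallic, specular and roughness values run through color-coded named reroutes into a SetMaterialAttributes node.
# The combined attributes then pass through the MF_TechArtOverrides_accessories material function.
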Material /BRCosmetics/Characters/Player/Male/Medium/Bodies/M_MED_Robot_01/Demo/Materials/Demo_M_MED_Robot_Backpack
graph {
 properties {
  bUseMaterialAttributes=true
  bUsedWithMeshDeformer=true
  bUsedWithSkeletalMesh=true
  pixelDepthOffsetMode=PDOM_Legacy
  usedShadingModels=Default Lit
 }
 nodes [
  n1 = TextureSample() {samplerType=SAMPLERTYPE_Normal, texture=/BRCosmetics/Characters/Player/Male/Medium/Bodies/M_MED_Robot_01/Demo/Textures/T_Robot_Backpack_N}
  n2 = SetMaterialAttributes!(
   MaterialAttributes
   Base Color=NamedRerouteDeclaration(None=TextureSample() {texture=/BRCosmetics/Characters/Player/Male/Medium/Bodies/M_MED_Robot_01/Demo/Textures/T_Robot_Backpack_D}.RGB) {name=baseColor, nodeColor=<rgba 1 0.1875 0 1>}
   Metallic=NamedRerouteDeclaration(None=Constant() {r=0.5}) {name=metallic, nodeColor=<rgba 1 0.6563 0 1>}
   Specular=NamedRerouteDeclaration(None=Constant() {r=1}) {name=specular, nodeColor=<rgba 0.3594 1 0 1>}
   Roughness=NamedRerouteDeclaration(None=Constant() {r=0.4}) {name=roughness, nodeColor=<rgba 0 1 0.8125 1>}
   Normal=NamedRerouteDeclaration(None=@n1.RGB) {name=normal, nodeColor=<rgba 0 0.4844 1 1>}
  ) {attributeSetTypes [
    #1
    #2
    #3
    #4
    #5
   ]}
  MaterialFunctionCall(In=@n2) {materialFunction=&MF_TechArtOverrides_accessories}
 ]
}
